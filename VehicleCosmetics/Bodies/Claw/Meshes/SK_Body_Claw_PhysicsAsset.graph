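# Physics asset for the Claw vehicle body cosmetic's skeletal mesh, with one capsule per bone for chassis, glass, suspension, brakes and wheels.
# Constraints with uniform 45-degree swing and twist limit values link glass and control arms to the body, front discs to steering knuckles, and rear discs, brake pads and shock bases to axle pivots.
# Steering knuckles and axle pivots have no constraint to the body, and front shock bases have none.
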PhysicsAsset /VehicleCosmetics/Bodies/Claw/Meshes/SK_Body_Claw_PhysicsAsset
bodies {
 axle_pivot_bk_l [{shapeType=Capsule, length=66.8542, radius=9.8073, center=<0.0054 0.994 0.3505>, rotation=<rot 0 -0.0032 -89.7418>}]
 axle_pivot_bk_r [{shapeType=Capsule, length=23.799, radius=9.8073, center=<0.0017 42.2149 -0.0568>, rotation=<rot 0 180 -86.0261>}]
 body [{shapeType=Capsule, length=266.3743, radius=116.7477, center=<23.4501 1.5109 26.7711>, rotation=<rot 0 -91.5283 -88.4023>}]
 brakepad_bk_l [{shapeType=Capsule, length=20.6224, radius=20.0921, center=<-4.1771 0.2407 -0.5476>, rotation=<rot -13.88 -90 -40.8258>}]
 brakepad_bk_r [{shapeType=Capsule, length=20.6224, radius=20.0921, center=<-4.1771 -0.2407 -0.5476>, rotation=<rot 13.88 -90 -40.8258>}]
 ctrlArm_bk_l [{shapeType=Capsule, length=25.4815, radius=18.4786, center=<-3.3066 -1.2258 -0.1249>, rotation=<rot 0 -124.306 -81.199>}]
 ctrlArm_bk_r [{shapeType=Capsule, length=25.4815, radius=18.4786, center=<-3.3066 1.2249 -0.1259>, rotation=<rot 0 -55.694 -81.199>}]
 ctrlArm_fr_l [{shapeType=Capsule, length=20.2501, radius=6.4033, center=<1.3952 -5.0578 -0.3383>, rotation=<rot 0 93.3738 -84.5005>}]
 ctrlArm_fr_r [{shapeType=Capsule, length=20.2501, radius=6.4033, center=<1.3946 5.0565 -0.337>, rotation=<rot 0 86.6262 -84.5005>}]
 glass_b [{shapeType=Capsule, length=65.7179, radius=45.236, center=<38.8425 -0.0831 10.9957>, rotation=<rot 0 179.8414 -89.945>}]
 glass_c [{shapeType=Capsule, length=86.8984, radius=44.5337, center=<-40.7425 0 18.1031>, rotation=<rot 0 180 -90>}]
 glass_l [{shapeType=Capsule, length=104.1981, radius=21.3446, center=<-46.4819 5.9594 16.7973>, rotation=<rot 0 -90.384 -88.9124>}]
 glass_r [{shapeType=Capsule, length=104.1981, radius=21.3446, center=<46.4819 -5.9594 -16.7973>, rotation=<rot 0 -90.384 -88.9124>}]
 shock_base_bk_l [{shapeType=Capsule, length=19.7089, radius=10.9446, center=<-1.5973 -1.8728 3.6051>, rotation=<rot 9.1828 -90 1.0683>}]
 shock_base_bk_r [{shapeType=Capsule, length=19.7089, radius=10.9446, center=<-1.5973 1.8729 3.6051>, rotation=<rot -9.1828 -90 1.0683>}]
 shock_base_fr_l [{shapeType=Capsule, length=20.0796, radius=14.9679, center=<-2.0447 3.1703 2.0464>, rotation=<rot 18.127 -90 1.9512>}]
 shock_base_fr_r [{shapeType=Capsule, length=20.0796, radius=14.9679, center=<-2.0447 -3.1703 2.0464>, rotation=<rot -18.127 -90 1.9512>}]
 wheel_disc_bk_l [{shapeType=Capsule, length=21.9157, radius=21.9154, center=<-0.0266 -0.8329 -0.0037>, rotation=<rot 0.0103 -90 0>}]
 wheel_disc_bk_r [{shapeType=Capsule, length=21.9157, radius=21.9154, center=<-0.0266 0.833 -0.0037>, rotation=<rot -0.0103 -90 0.0003>}]
 wheel_disc_fr_l [{shapeType=Capsule, length=21.9157, radius=21.9154, center=<-0.0096 -0.7665 0.0397>, rotation=<rot 0.0102 -90 0>}]
 wheel_disc_fr_r [{shapeType=Capsule, length=21.9157, radius=21.9154, center=<-0.0096 0.7665 0.0397>, rotation=<rot -0.0102 -90 0.0009>}]
 wheel_steering_fr_l [{shapeType=Capsule, length=20.6086, radius=20.0248, center=<-4.0977 -3.5192 -0.4104>, rotation=<rot -14.452 -90 -42.9127>}]
 wheel_steering_fr_r [{shapeType=Capsule, length=20.6086, radius=20.0248, center=<-4.0977 3.5192 -0.4104>, rotation=<rot 14.452 -90 -42.9127>}]
}
constraints [
 {bone1Name=glass_c, bone2Name=body, swing1LimitDegrees=45, swing2LimitDegrees=45, twistLimitDegrees=45}
 {bone1Name=glass_l, bone2Name=body, swing1LimitDegrees=45, swing2LimitDegrees=45, twistLimitDegrees=45}
 {bone1Name=glass_r, bone2Name=body, swing1LimitDegrees=45, swing2LimitDegrees=45, twistLimitDegrees=45}
 {bone1Name=glass_b, bone2Name=body, swing1LimitDegrees=45, swing2LimitDegrees=45, twistLimitDegrees=45}
 {bone1Name=ctrlArm_fr_l, bone2Name=body, swing1LimitDegrees=45, swing2LimitDegrees=45, twistLimitDegrees=45}
 {bone1Name=ctrlArm_fr_r, bone2Name=body, swing1LimitDegrees=45, swing2LimitDegrees=45, twistLimitDegrees=45}
 {bone1Name=ctrlArm_bk_l, bone2Name=body, swing1LimitDegrees=45, swing2LimitDegrees=45, twistLimitDegrees=45}
 {bone1Name=ctrlArm_bk_r, bone2Name=body, swing1LimitDegrees=45, swing2LimitDegrees=45, twistLimitDegrees=45}
 {bone1Name=wheel_disc_fr_l, bone2Name=wheel_steering_fr_l, swing1LimitDegrees=45, swing2LimitDegrees=45, twistLimitDegrees=45}
 {bone1Name=wheel_disc_fr_r, bone2Name=wheel_steering_fr_r, swing1LimitDegrees=45, swing2LimitDegrees=45, twistLimitDegrees=45}
 {bone1Name=wheel_disc_bk_l, bone2Name=axle_pivot_bk_l, swing1LimitDegrees=45, swing2LimitDegrees=45, twistLimitDegrees=45}
 {bone1Name=brakepad_bk_l, bone2Name=axle_pivot_bk_l, swing1LimitDegrees=45, swing2LimitDegrees=45, twistLimitDegrees=45}
 {bone1Name=shock_base_bk_l, bone2Name=axle_pivot_bk_l, swing1LimitDegrees=45, swing2LimitDegrees=45, twistLimitDegrees=45}
 {bone1Name=wheel_disc_bk_r, bone2Name=axle_pivot_bk_r, swing1LimitDegrees=45, swing2LimitDegrees=45, twistLimitDegrees=45}
 {bone1Name=brakepad_bk_r, bone2Name=axle_pivot_bk_r, swing1LimitDegrees=45, swing2LimitDegrees=45, twistLimitDegrees=45}
 {bone1Name=shock_base_bk_r, bone2Name=axle_pivot_bk_r, swing1LimitDegrees=45, swing2LimitDegrees=45, twistLimitDegrees=45}
]
